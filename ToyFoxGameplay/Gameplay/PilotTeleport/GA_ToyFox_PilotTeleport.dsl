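; Gameplay ability that presents the ToyFox pilot teleport with a player montage, camera animation, camera modifier and force feedback.
; Activation pushes a default hipfire camera mode for alignment and exits hiding props.
; Montage notify tags hide the pawn without collision, then fade the camera to white.
; The authority optionally force-revives from DBNO via a hotfix float, and ending the ability reverts camera, visibility, fade and montage task.
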(event Ability|EventActivateAbility
  ;[0.00079 0.15 0.02893 0.5] ActivateAbility: Play the teleport player animation and camera animations.
  (|CallParentFunction)
  (CallFunction|PlayCameraAnimation)
  (CallFunction|SetEnableCameraModifier true)
  (Utilities|FlowControl|Sequence
    (:then_0
      (bind (_asynctaskproxy _notifytag) (Ability|Tasks|PlayMontageAndWaitNotify "MontageTask" (Variables|Settings|GetAnimMontage))
        (:then
          (Variables|Transient|SetAsyncTask_PlayMontage _asynctaskproxy)
          (bind _outfortplayerpawn (CallFunction|GetAvatarasValidFortPlayerPawn)
            (:ValidFortPlayerPawn
              (Utilities|FlowControl|Sequence
                (:then_0
                  (Utilities|IsValid (Fortnite|GetFortPlayerControllerfromActor _outfortplayerpawn)
                    (:Is%20Valid
                      (Game|Feedback|LocalPlayForceFeedback (Fortnite|GetFortPlayerControllerfromActor _outfortplayerpawn) (Variables|ForceFeedback|GetPilotTeleportForceFeedback) false "None"))))
                (:then_1
                  (if (Class|FortPawn|IsInsideHidingProp _outfortplayerpawn true)
                    (Ability|SendGameplayEvent "(TagName=\"Event.Athena.HidingProp.StopHiding\")" (Utilities|Struct|MakeGameplayEventData "(TagName=\"\")"))))))))
        (:OnNotifyBegin
          (switch Utilities|FlowControl|Switch|SwitchonGameplayTag _notifytag
            (:Targeting.ApplicationTag.0
              (CallFunction|SetPawnHiddenandDisableCollision true))
            (:Targeting.ApplicationTag.1
              (CallFunction|SetFadetoWhite true))))))
    (:then_1
      (Ability|HasAuthority
        (:True
          (if (> (ScalableFloat|GetCachedScalableFloatValue (Variables|Settings|GetHotfix_ForceReviveFromDBNO)) 0)
            (bind _asfort_player_pawn_athena (Utilities|Casting|CastToFortPlayerPawnAthena (CallFunction|GetAvatarasFortPlayerPawn)))
            (Class|FortPlayerPawnAthena|ForceRevivefromDBNO _asfort_player_pawn_athena (Pawn|GetController _asfort_player_pawn_athena))))))))

(event Ability|EventOnEndAbility
  (params (bWasCancelled :type Boolean))
  ;[0.1094 0.06617 0 0.5] OnEndAbility: When the ability ends, stop all the camera animations and revert player state.
  (Utilities|FlowControl|Sequence
    (:then_0
      (bind _playercameraanimationcameramodifier (Variables|Transient|GetPlayerCameraAnimationCameraModifier)
        (:then
          (CameraAnimation|StopCameraAnimation _playercameraanimationcameramodifier (Variables|Transient|GetCameraAnimationHandle) false :!class /Script/EngineCameras.CameraAnimationCameraModifier))
        (:else
          _))
      (Camera|PopCameraMode)
      (CallFunction|SetPawnHiddenandDisableCollision false)
      (CallFunction|SetEnableCameraModifier false)
      (CallFunction|SetFadetoWhite false)
      (bind _asynctask_playmontage (Variables|Transient|GetAsyncTask_PlayMontage)
        (:then
          (GameplayTasks|EndTask _asynctask_playmontage))
        (:else)))
    (:then_1
      (|CallParentFunction bWasCancelled))))

(fn CanActivateAbility
  (params (ActorInfo :type Gameplay%20Ability%20Actor%20Info%20Structure) (Handle :type Gameplay%20Ability%20Spec%20Handle%20Structure))
  (return 0 true))

(fn PlayCameraAnimation
  (bind _cameraanimation (Variables|Settings|GetCameraAnimation)
    (:then
      (bind _playercontroller (Variables|Transient|GetPlayerController)
        (:then
          (bind _playercameramanager (Variables|/Script/Engine.PlayerController|GetPlayerCameraManager _playercontroller)
            (:then
              (bind _output_get (Variables|Transient|SetPlayerCameraAnimationCameraModifier (CameraAnimation|ConvCameraAnimationCameraModifier _playercameramanager)))
              (Utilities|IsValid _output_get
                (:Is%20Valid
                  ;[0.15 0.15 0.5 0.5] Push default camera so cam anim lines up
                  (Camera|PushCameraMode "/Game/Blueprints/Camera/Athena/Athena_PlayerCameraModeRanged_Hipfire_V2.Athena_PlayerCameraModeRanged_Hipfire_V2_C")
                  (bind _returnvalue (CameraAnimation|PlayCameraAnimation _output_get _cameraanimation (Utilities|Struct|MakeCameraAnimationParams 1.0) :!class /Script/EngineCameras.CameraAnimationCameraModifier))
                  (Variables|Transient|SetCameraAnimationHandle _returnvalue))))
            (:else)))
        (:else)))
    (:else)))

(fn SetEnableCameraModifier
  (params (bEnable :type Boolean))
  (bind _playercontroller (Variables|Transient|GetPlayerController)
    (:then
      (bind _playercameramanager (Variables|/Script/Engine.PlayerController|GetPlayerCameraManager _playercontroller)
        (:then
          (if bEnable
            (bind _returnvalue (Game|Player|AddNewCameraModifier _playercameramanager (Variables|Settings|GetCameraModifierClass)))
            (Variables|Transient|SetCameraModifier _returnvalue)
            (else
              (Game|Player|RemoveCameraModifier _playercameramanager (Variables|Transient|GetCameraModifier)))))
        (:else)))
    (:else)))

(fn SetFadeToWhite
  (params (bEnable :type Boolean))
  (bind _playercontroller (Variables|Transient|GetPlayerController)
    (:then
      (bind _playercameramanager (Variables|/Script/Engine.PlayerController|GetPlayerCameraManager _playercontroller)
        (:then
          (if bEnable
            (CameraFades|StartCameraFade _playercameramanager 0.0 1.0 0.25 "(R=1.000000,G=1.000000,B=1.000000,A=1.000000)" false true)
            (else
              (CameraFades|StopCameraFade _playercameramanager))))
        (:else)))
    (:else)))
