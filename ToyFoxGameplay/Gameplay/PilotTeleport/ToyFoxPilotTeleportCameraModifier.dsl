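; Camera modifier for pilot teleport that captures a yaw-only target camera rotation when it is added to a camera.
; On being added, it reads the camera manager's current rotation, zeroes pitch and roll, and stores the result.
; A getter function exposes the transient target camera rotation.
; The graph shown contains no per-frame camera adjustment.
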(event EventBlueprintAddedToCamera
  (params (Camera :type Player%20Camera%20Manager%20Object%20Reference))
  (bind (_returnvalue_roll _returnvalue_pitch _returnvalue_yaw) (Math|Rotator|BreakRotator (Camera|GetCameraRotation Camera)))
  (Variables|SetTargetCameraRotation_Roll (Math|Rotator|MakeRotator 0.0 0.0 _returnvalue_yaw)))

(fn GetTargetCameraRotation
  (return (Variables|Transient|GetTargetCameraRotation)))
